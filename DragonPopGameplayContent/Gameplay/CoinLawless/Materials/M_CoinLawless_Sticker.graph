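# Masked, two-sided Default Lit material for the CoinLawless sticker, flagged for use with clothing and Nanite.
# The sticker color texture is halved and divided by EyeAdaptation, a technique that offsets auto-exposure.
# A separate alpha texture is sampled alongside it, and both samples use shared clamped samplers.
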Material /DragonPopGameplayContent/Gameplay/CoinLawless/Materials/M_CoinLawless_Sticker
graph {
 properties {
  bAutomaticallySetUsageInEditor=false
  bUsedWithClothing=true
  bUsedWithNanite=true
  blendMode=BLEND_Masked
  twoSided=true
  usedShadingModels=Default Lit
 }
 nodes [
  Divide(
   A=Multiply(A=TextureSample() {samplerSource=SSM_Clamp_WorldGroupSettings, texture=/DragonPopGameplayContent/Gameplay/CoinLawless/Textures/T_Sticker_CoinLawless}.RGB) {constB=0.5}
   B=EyeAdaptation().EyeAdaptation
  )
  TextureSample() {samplerSource=SSM_Clamp_WorldGroupSettings, texture=/DragonPopGameplayContent/Gameplay/CoinLawless/Textures/T_Sticker_CoinLawless_Alpha}
 ]
}
