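# Surface material for the Lawless coin pickup: lit base textures routed through a four-layer thin-film function, with animated emissive detail.
# Texture masks and radial distance from pre-skinned local position separate rings, sides and emblem areas.
# Time-driven thresholds, sine waves and parallax cell patterns build the glow, two-color tinted and divided by eye adaptation.
# An optional WPO switch spins and bobs the coin, correcting normals, with separate case and pickup height settings.
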Material /DragonPopGameplayContent/Gameplay/CoinLawless/Materials/M_Pickup_CoinLawless
graph {
 properties {
  bAutomaticallySetUsageInEditor=false
  bTangentSpaceNormal=false
  bUseMaterialAttributes=true
  previewMesh=/DragonPopGameplayContent/Gameplay/CoinLawless/Meshes/CoinLawless_Pickup
  usedShadingModels=Default Lit
 }
 nodes [
  n1 = TextureSampleParameter2D() {group=Coin, parameterName=Diffuse, sortPriority=0, texture=/DragonPopGameplayContent/Gameplay/CoinLawless/Textures/T_Pickup_CoinLawless_D}
  n2 = NamedRerouteDeclaration(None=@n1.RGB) {name=T_In_Basecolor, nodeColor=<rgba 0 0.9062 1 1>}
  Add(A=Multiply(B=NamedRerouteUsage()) {constA=0.2}, B=@n2)
  n3 = NamedRerouteUsage()
  n4 = Step(Y=NamedRerouteUsage()) {constX=0.615}
  n5 = Saturate(None=Multiply(A=Subtract(A=@n3, B=@n4), B=NamedRerouteUsage()))
  Add(A=NamedRerouteUsage(), B=Multiply(A=NamedRerouteDeclaration(None=@n5) {name=M_GreenRing, nodeColor=<rgba 1 0 0.0234 1>}) {constB=0.1})
  n6 = LocalPosition() {localOrigin=InstancePreSkinning}
  n7 = ShaderStageSwitch(PixelShader=VertexInterpolator(VS=@n6.XYZ).PS, VertexShader=@n6.XYZ)
  n8 = Constant3Vector() {constant=<rgba 0.05 0 -0.6 0>}
  n9 = Distance(A=NamedRerouteDeclaration(None=Add(A=@n7, B=@n8)) {name=Preskinned Pos, nodeColor=<rgba 0 1 0.4141 1>}, B=Constant3Vector())
  n10 = NamedRerouteDeclaration(None=Saturate(None=Divide(A=@n9) {constB=15})) {name=M_Coin_Radial, nodeColor=<rgba 0.7109 1 0 1>}
  n11 = LinearInterpolate(A=@n10, B=OneMinus(None=@n10), Alpha=Step(Y=NamedRerouteUsage()) {constX=0.6})
  n12 = ScalarParameter() {defaultValue=4.3, group=Coin, parameterName=Emissive_Tile}
  n13 = Multiply(A=Time(), B=ScalarParameter() {defaultValue=-0.23, group=Coin, parameterName=Emissive_Speed})
  n14 = NamedRerouteUsage()
  n15 = ScalarParameter() {defaultValue=0.2, group=Coin, parameterName=Emissive_Speed_Random}
  n16 = Add(A=@n13, B=Multiply(A=@n13, B=Multiply(A=@n14, B=@n15)))
  n17 = Add(
   A=Add(A=Multiply(A=@n11, B=@n12) {constB=4}, B=@n16)
   B=Multiply(A=NamedRerouteUsage(), B=ScalarParameter() {defaultValue=31.45, group=Coin, parameterName=Emissive_Position_Random})
  )
  n18 = SmoothStep(
   Min=ScalarParameter() {defaultValue=0.8, group=Coin, parameterName=Emissive_min}
   Max=ScalarParameter() {defaultValue=1, group=Coin, parameterName=Emissive_max}
   Value=MaterialFunctionCall!(
    ThresholdValue=ScalarParameter() {defaultValue=0.5, group=Coin, parameterName=Emissive_Treshold}
    ThresholdRange=ScalarParameter() {defaultValue=0.5, group=Coin, parameterName=Emissive_TresholdFalloff}
    Alpha=Frac(None=@n17)
    TwoSidedThreshold=StaticBool() {value=true}
   ) {materialFunction=/Engine/Functions/Engine_MaterialFunctions02/Masking/ThresholdWithRange}.Result
  )
  n19 = SmoothStep(Value=NamedRerouteUsage()) {constMax=0.85, constMin=0.65}
  n20 = SmoothStep(Value=NamedRerouteUsage()) {constMin=0.5}
  n21 = Multiply(A=@n20, B=@n20)
  n22 = Sine(None=NamedRerouteUsage())
  n23 = NamedRerouteUsage()
  n24 = LinearInterpolate(A=Constant() {r=0.05}, B=Constant() {r=0.08}, Alpha=Multiply(A=ConstantBiasScale(None=@n22), B=Multiply(A=@n23) {constB=10}))
  n25 = Multiply(A=Max(A=Multiply(A=@n18, B=@n19), B=Multiply(A=@n21, B=@n24)), B=OneMinus(None=NamedRerouteUsage()))
  n26 = NamedRerouteUsage()
  n27 = OneMinus(None=NamedRerouteUsage())
  n28 = Multiply(A=Multiply(A=SmoothStep(Value=@n27) {constMax=0.3, constMin=0.15}, B=SmoothStep(Value=@n27) {constMax=0.3, constMin=0.65}))
  n29 = Multiply(A=Multiply(A=Multiply(A=@n25, B=@n26), B=Saturate(None=@n28)), B=NamedRerouteUsage())
  n30 = NamedRerouteUsage()
  n31 = Multiply(A=@n30, B=@n30)
  n32 = Constant() {r=0.03}
  n33 = NamedRerouteUsage()
  n34 = Multiply(
   A=Multiply(A=NamedRerouteUsage(), B=ScalarParameter() {defaultValue=0.5, group=Coin, parameterName=Emissive_Mult})
   B=LinearInterpolate(B=ConstantBiasScale(None=@n33), Alpha=ScalarParameter() {defaultValue=1, group=Coin, parameterName=Emissive_Pattern}) {constA=1}
  )
  Add(A=Add(A=@n29, B=Multiply(A=@n31, B=@n32)), B=@n34)
  n35 = OneMinus(None=Saturate(None=NamedRerouteUsage()))
  n36 = NamedRerouteUsage()
  Add(A=Multiply(A=Add(A=@n35, B=@n36), B=NamedRerouteUsage()), B=Multiply(A=NamedRerouteUsage()))
  n37 = NamedRerouteUsage()
  Add(A=@n37) {constB=-2}
  n38 = Max(A=NamedRerouteUsage(), B=Multiply(A=NamedRerouteUsage()) {constB=0.5})
  n39 = NamedRerouteUsage()
  n40 = NamedRerouteUsage()
  n41 = NamedRerouteDeclaration(None=Subtract(A=Max(A=@n38, B=@n39), B=Multiply(A=@n40) {constB=0.25})) {name=M_ElementsValues, nodeColor=<rgba 1 0.6094 0 1>}
  n42 = Multiply(A=@n37) {constB=2}
  n43 = Time()
  n44 = ScalarParameter() {defaultValue=1.35, desc=1.35, group=Coin, parameterName=Wave_Speed}
  n45 = Add(A=Multiply(A=Add(A=@n41, B=@n42), B=Constant() {r=0.1}), B=Add(A=Divide(A=@n43, B=@n44)) {constB=-0.5})
  n46 = NamedRerouteUsage()
  n47 = MaterialFunctionCall(
   ThresholdValue=Constant() {r=0.5}
   ThresholdRange=Constant() {r=0.8}
   Alpha=Frac(None=Add(A=@n45, B=@n46))
   TwoSidedThreshold=StaticBool() {value=true}
  ) {materialFunction=&ThresholdWithRange}
  Add(A=@n47.Result) {constB=0.1}
  n48 = Add(A=Time()) {constB=-0.25}
  Add(A=Constant() {r=0.3}, B=Multiply(A=Sine(None=@n48) {period=1.35}) {constB=0.2})
  n49 = Multiply(A=Add(A=SmoothStep(Value=@n36) {constMax=0.4}) {constB=0}) {constB=0.05}
  n50 = NamedRerouteUsage()
  n51 = Constant() {r=-0.6}
  n52 = Saturate(None=Multiply(A=Add(A=@n50, B=@n51), B=Constant() {r=-8}))
  n53 = NamedRerouteUsage()
  LinearInterpolate(A=@n29, B=Multiply(A=Multiply(A=@n49, B=@n52), B=OneMinus(None=@n53)) {constB=0.1}, Alpha=NamedRerouteUsage()) {constB=0}
  n54 = NamedRerouteUsage()
  n55 = TextureSampleParameter2D() {
   group=Coin
   parameterName=SpecularMasks
   samplerType=SAMPLERTYPE_Masks
   sortPriority=0
   texture=/DragonPopGameplayContent/Gameplay/CoinLawless/Textures/T_Pickup_CoinLawless_S
  }
  n56 = NamedRerouteDeclaration(None=@n55.G) {name=T_In_Metallic, nodeColor=<rgba 1 0.7031 0 1>}
  n57 = NamedRerouteDeclaration(None=@n55.B) {name=T_IN_Roughtness, nodeColor=<rgba 0 0.9766 1 1>}
  n58 = SetMaterialAttributes!(
   MaterialAttributes
   Base Color=@n2
   Emissive Color=@n54
   Specular=@n55.R
   Metallic=@n56
   Roughness=@n57
   Normal=NamedRerouteUsage()
   World Position Offset=NamedRerouteUsage()
  ) {attributeSetTypes [
    #1
    #2
    #3
    #4
    #5
    #6
    #7
   ]}
  n59 = TextureSampleParameter2D() {
   group=Coin
   parameterName=Normals
   samplerType=SAMPLERTYPE_Normal
   sortPriority=0
   texture=/DragonPopGameplayContent/Gameplay/CoinLawless/Textures/T_Pickup_CoinLawless_N
  }
  n60 = SetMaterialAttributes(
   Base Color=@n2
   Emissive Color=@n54
   Specular=@n55.R
   Metallic=@n56
   Roughness=@n57
   Normal=Transform(None=NamedRerouteDeclaration(None=@n59.RGB) {name=T_IN_Normals, nodeColor=<rgba 1 0.9609 0 1>})
  ) {attributeSetTypes [
    #1
    #2
    #3
    #4
    #5
    #6
    #7
   ]}
  MaterialFunctionCall!(
   In=StaticSwitch(True=@n58, False=@n60, Value=StaticBoolParameter() {defaultValue=true, group=Coin, parameterName=WPO, sortPriority=1})
   SFXMask=NamedRerouteUsage()
  ) {desc=Normals are transformed from tangent to WS, materialFunction=/Game/Global/Base4Layers_2022/MaterialFunctions/MF_4Layer_ThinFilm}
  Multiply(A=NamedRerouteUsage())
  Multiply(A=@n35, B=Saturate(None=@n36))
  Multiply(A=@n55.B)
  n61 = Add(A=TextureCoordinate(), B=AppendVector(A=Constant() {r=0.013}, B=Constant() {r=0.025}))
  n62 = Constant() {r=2}
  n63 = TextureSampleParameter2D(UVs=NamedRerouteDeclaration(None=Multiply(A=@n61, B=@n62)) {name=CoinUV Center, nodeColor=<rgba 0 1 0.7422 1>}) {
   group=Coin
   parameterName=Masks_TA
   samplerType=SAMPLERTYPE_LinearColor
   texture=/DragonPopGameplayContent/Gameplay/CoinLawless/Textures/PickleCoin_TA_Masks
  }
  NamedRerouteDeclaration(None=@n63.G) {name=T_IN_SDF, nodeColor=<rgba 0.5 0 1 1>}
  NamedRerouteDeclaration(None=@n63.B) {name=T_IN_Random, nodeColor=<rgba 0.5 0 1 1>}
  NamedRerouteDeclaration(None=@n16) {name=S_Time_Random, nodeColor=<rgba 0.9453 1 0 1>}
  n64 = TextureSampleParameter2D() {group=Coin, parameterName=FX Masks, sortPriority=0, texture=/DragonPopGameplayContent/Gameplay/CoinLawless/Textures/T_Pickup_CoinLawless_FX}
  NamedRerouteDeclaration(None=@n64.RGB) {name=SFX Mask, nodeColor=<rgba 0 1 0.25 1>}
  n65 = NamedRerouteUsage()
  NamedRerouteDeclaration(None=Step(X=ComponentMask(None=@n65) {g=true}) {constX=0}) {name=M_CoinSide_A, nodeColor=<rgba 1 0.0938 0 1>}
  n66 = MaterialFunctionCall(Color=Divide(A=ObjectPositionWS()) {constB=3.14}) {materialFunction=&SplitComponents}
  NamedRerouteDeclaration(None=Add(A=@n66.R, B=@n66.R)) {name=World Pos, nodeColor=<rgba 1 0 0.7734 1>}
  n67 = Subtract(
   A=Multiply(A=NamedRerouteUsage(), B=ScalarParameter() {defaultValue=-0.5, group=Coin, parameterName=Wave_Freq})
   B=Divide(A=Time(), B=ScalarParameter() {defaultValue=1.35, desc=1.35, group=Coin, parameterName=Wave_Speed})
  )
  n68 = NamedRerouteUsage()
  n69 = SmoothStep(Min=ScalarParameter() {defaultValue=-1, group=Coin, parameterName=Wave_Bias}, Value=Sine(None=Add(A=@n67, B=@n68))) {constMin=0.1}
  NamedRerouteDeclaration(None=@n69) {name=M_Wave, nodeColor=<rgba 1 0.1875 0 1>}
  n70 = Fresnel(ExponentIn=ScalarParameter() {defaultValue=2, group=ThermiteHologram, parameterName=FresnelExponent}) {baseReflectFraction=0.01}
  n71 = Normalize(VectorInput=VectorParameter() {group=ThermiteHologram, parameterName=SweepDirectionCamSpace, defaultValue=<rgba -0.5 0.45 -1 1>}.RGB)
  n72 = Multiply(A=DotProduct(A=LocalPosition().XYZ, B=@n71), B=@n71)
  n73 = ComponentMask(None=Panner(
    Coordinate=ComponentMask(None=@n72) {b=true, r=true}
    Time=Time()
    Speed=ScalarParameter() {defaultValue=100, group=ThermiteHologram, parameterName=Speed}
   ) {speedX=0.2, speedY=0.2}) {g=true}
  n74 = ScalarParameter() {defaultValue=200, group=ThermiteHologram, parameterName=Freq}
  n75 = SmoothStep(
   Min=ScalarParameter() {defaultValue=0.1, group=ThermiteHologram, parameterName=Sheen Size}
   Max=Constant()
   Value=Frac(None=Divide(A=@n73, B=@n74))
  )
  NamedRerouteDeclaration(None=Multiply(
    A=NamedRerouteDeclaration(None=@n70) {name=Fresnel_simple, nodeColor=<rgba 0 0.5547 1 1>}
    B=NamedRerouteDeclaration(None=@n75) {name=M_Sheen, nodeColor=<rgba 0 1 0.4141 1>}
   )) {name=M_Sheen_fresnel, nodeColor=<rgba 0 0.8359 1 1>}
  n76 = Multiply(A=Step(X=@n63.R) {constY=0.2}, B=Step(Y=NamedRerouteUsage()) {constX=0.74})
  NamedRerouteDeclaration(None=Multiply(A=@n76, B=NamedRerouteUsage())) {name=M_Pickle, nodeColor=<rgba 0.3828 1 0 1>}
  n77 = Saturate(None=Multiply(A=Subtract(A=@n47.Result) {constB=0.8}) {constB=5})
  n78 = OneMinus(None=NamedRerouteUsage())
  NamedRerouteDeclaration(None=Multiply(A=Multiply(A=@n77, B=@n78), B=@n52)) {name=Anim_SIlhouete, nodeColor=<rgba 0.9219 0 1 1>}
  n79 = Time()
  n80 = Sine(None=Multiply(A=Multiply(A=@n79) {constB=1.5}, B=ScalarParameter() {defaultValue=0.08, desc=0.08, group=Coin, parameterName=WPO_Speed}))
  n81 = Constant3Vector() {constant=<rgba 0 0 3.8 0>}
  n82 = Multiply(
   A=Add(A=@n80, B=ScalarParameter() {defaultValue=1.8, group=Coin, parameterName=WPO_Vertical_Case_Bias}) {constB=1.8}
   B=Multiply(A=@n81, B=ScalarParameter() {defaultValue=1, group=Coin, parameterName=WPO_Vertical_Case_Mult})
  )
  n83 = Constant3Vector() {constant=<rgba 0 0 1 0>}
  n84 = MaterialFunctionCall!(
   RotationAngle=Multiply(A=Time(), B=ScalarParameter() {defaultValue=0.08, desc=0.08, group=Coin, parameterName=WPO_Speed}) {constB=0}
   NormalizedRotationAxis=Normalize(VectorInput=Transform(None=@n83) {transformSourceType=TRANSFORMSOURCE_Local})
   PivotPoint=TransformPosition(None=Constant3Vector()) {transformType=TRANSFORMPOSSOURCE_World}
   Position=WorldPosition().XYZ
   Normal=Transform(None=NamedRerouteUsage())
   TangentU
   TangentV
  ) {materialFunction=/Game/Characters/Player/Male/Medium/Bodies/M_MED_Inferno/FX/Mat/RotateVertexAboutAxis}
  n85 = Multiply(
   A=Add(A=@n80, B=ScalarParameter() {defaultValue=0.5, group=Coin, parameterName=WPO_Vertical_Pickup_Bias}) {constB=1.8}
   B=Multiply(A=@n81, B=ScalarParameter() {defaultValue=0.6, group=Coin, parameterName=WPO_Vertical_Pickup_Mult})
  )
  NamedRerouteDeclaration(None=StaticSwitch(True=Add(A=@n82, B=@n84.WPO), False=Add(A=@n84.WPO, B=@n85), Value=StaticBoolParameter() {group=Coin, parameterName=Is_Case, sortPriority=1})) {name=WPO, nodeColor=<rgba 1 0 0.0703 1>}
  NamedRerouteDeclaration(None=@n84.Normal) {name=Normals, nodeColor=<rgba 0.6406 0 1 1>}
  n86 = BumpOffset(Coordinate=ComponentMask(None=NamedRerouteUsage()) {b=true, r=true}, Height=Constant() {r=-75})
  n87 = TextureSample(UVs=Divide(A=@n86, B=ScalarParameter() {defaultValue=30, group=Coin, parameterName=Cell_Tile})) {
   samplerType=SAMPLERTYPE_LinearGrayscale
   texture=/Game/Environments/Helios/Props/Quest/Quest_RiftBeacon_A/Textures/T_Quest_RiftBeacon_Panels_A_A_D_Cells
  }
  n88 = Add(A=@n87.R, B=@n69)
  NamedRerouteDeclaration(None=@n88) {name=M_Pattern, nodeColor=<rgba 0 1 0.6484 1>}
  n89 = Step(Y=NamedRerouteUsage()) {constX=0.545}
  NamedRerouteDeclaration(None=@n89) {name=M_CoinRing_Small, nodeColor=<rgba 0 1 0.6719 1>}
  n90 = Constant() {r=1.5}
  n91 = Multiply(A=Time()) {constB=0.05}
  n92 = Divide(A=@n91) {constB=0.34}
  n93 = Frac(None=Add(A=AppendVector(A=@n90, B=@n92), B=Divide(A=@n86) {constB=7}))
  n94 = Constant()
  n95 = Frac(None=Add(A=AppendVector(A=@n94, B=@n91), B=Divide(A=@n86) {constB=10}))
  NamedRerouteDeclaration(None=Multiply(
    A=TextureSample(UVs=@n93) {samplerType=SAMPLERTYPE_LinearColor, texture=/Game/Global/Textures/Default/Noise/T_Atlas_DigitalNoise_Small}.G
    B=TextureSample(UVs=@n95) {samplerType=SAMPLERTYPE_LinearColor, texture=&T_Atlas_DigitalNoise_Small}.B
   )) {name=Pattern_2, nodeColor=<rgba 1 0.4219 0 1>}
  n96 = ComponentMask(None=NamedRerouteUsage()) {g=true}
  n97 = NamedRerouteUsage()
  n98 = NamedRerouteUsage()
  n99 = Multiply(
   A=Multiply(A=Step(X=@n96) {constY=0.1}, B=Step(Y=@n97) {constX=0.5})
   B=NamedRerouteDeclaration(None=OneMinus(None=@n98)) {name=M_CoinSide_B, nodeColor=<rgba 1 0.7734 0 1>}
  )
  NamedRerouteDeclaration(None=Multiply(A=NamedRerouteDeclaration(None=@n99) {name=M_D$, nodeColor=<rgba 0.3828 0 1 1>}, B=SmoothStep(Value=@n96) {constMax=0.5, constMin=0.15})) {name=M_D_Details, nodeColor=<rgba 1 0.3516 0 1>}
  n100 = Saturate(None=@n34)
  n101 = SmoothStep(Value=@n100) {constMax=0.05}
  n102 = LinearInterpolate(
   A=VectorParameter() {group=Coin, parameterName=Emissive_Color2, defaultValue=<rgba 0 0.1 0.0411 1>}.RGB
   B=VectorParameter() {group=Coin, parameterName=Emissive_Color1_Close, defaultValue=<rgba 0.6578 2.1 0 1>}.RGB
   Alpha=@n100
  )
  NamedRerouteDeclaration(None=Divide(A=Multiply(A=@n101, B=@n102), B=EyeAdaptation().EyeAdaptation)) {name=Emissive, nodeColor=<rgba 0.0781 0 1 1>}
  n103 = NamedRerouteUsage()
  n104 = NamedRerouteUsage()
  n105 = Max(A=OneMinus(None=NamedRerouteUsage()), B=Subtract(A=Subtract(A=@n103, B=@n104), B=@n89))
  n106 = NamedRerouteUsage()
  NamedRerouteDeclaration(None=Saturate(None=Subtract(A=@n105, B=@n106))) {name=M_Gold, nodeColor=<rgba 0 0.0625 1 1>}
  n107 = OneMinus(None=NamedRerouteUsage())
  n108 = NamedRerouteDeclaration(None=Step(Y=NamedRerouteUsage()) {constX=0.67}) {name=M_CoinRing_A, nodeColor=<rgba 0 0.0625 1 1>}
  NamedRerouteDeclaration(None=LinearInterpolate(A=Multiply(A=@n107, B=@n89), B=Subtract(A=@n108, B=@n76), Alpha=NamedRerouteUsage())) {name=M_BG, nodeColor=<rgba 1 0 0.5859 1>}
  SmoothStep() {constMax=0.95, constMin=1.5}
  SmoothStep(Value=@n88) {constMax=0.95, constMin=1.5}
 ]
}
